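# Sound cue for the Carmine punch swing, layering close and distant variants, each with its own attenuation asset.
# A quality-level switch sends its first two slots to full waves and its last two to "_opt" waves.
# Each layer randomly picks one of four waves, varies pitch slightly and overrides its sound class to third-person ranged weapon fire.
# Concurrency stops the oldest instance beyond two.
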SoundCue /Game/Sounds/Fort_GamePlay_Sounds/Carmine/Carmine_Punch_Swing_Cue
root=QualityLevel_0
nodes [
 {
  id=Attenuation_0
  children [Modulator_2]
  properties {attenuationSettings=/Game/Sounds/Attenuation/Fort_Weapon_Carmine_Close_Athena_Att, attenuationOverrides {bSpatialize=false}}
 }
 {
  id=Attenuation_1
  children [Modulator_0]
  properties {attenuationSettings=&Fort_Weapon_Carmine_Close_Athena_Att, attenuationOverrides {bSpatialize=false}}
 }
 {
  id=Attenuation_2
  children [Modulator_1]
  properties {attenuationSettings=/Game/Sounds/Attenuation/Fort_Weapon_Carmine_Far_Athena_Att, attenuationOverrides {bSpatialize=false}}
 }
 {
  id=Attenuation_3
  children [Modulator_3]
  properties {attenuationSettings=&Fort_Weapon_Carmine_Far_Athena_Att, attenuationOverrides {bSpatialize=false}}
 }
 {id=Mixer_0, children [SoundClass_0,SoundClass_1], properties {inputVolume [1,1]}}
 {id=Mixer_1, children [SoundClass_2,SoundClass_3], properties {inputVolume [1,1]}}
 {id=Modulator_0, children [Random_1], properties {pitchMax=1.1, pitchMin=0.9}}
 {id=Modulator_1, children [Random_0], properties {pitchMax=1.1, pitchMin=0.9}}
 {id=Modulator_2, children [Random_2], properties {pitchMax=1.1, pitchMin=0.9}}
 {id=Modulator_3, children [Random_3], properties {pitchMax=1.1, pitchMin=0.9}}
 {id=QualityLevel_0, children [Mixer_0,Mixer_0,Mixer_1,Mixer_1]}
 {id=Random_0, children [WavePlayer_4,WavePlayer_5,WavePlayer_6,WavePlayer_7], properties {weights [1,1,1,1]}}
 {id=Random_1, children [WavePlayer_0,WavePlayer_1,WavePlayer_2,WavePlayer_3], properties {weights [1,1,1,1]}}
 {id=Random_2, children [WavePlayer_8, WavePlayer_9, WavePlayer_10, WavePlayer_11], properties {weights [1,1,1,1]}}
 {id=Random_3, children [WavePlayer_12, WavePlayer_13, WavePlayer_14, WavePlayer_15], properties {weights [1,1,1,1]}}
 {
  id=SoundClass_0
  children [Attenuation_1]
  properties {soundClassOverride=&Fort_WeaponRanged_Fire_3P_PSM}
 }
 {
  id=SoundClass_1
  children [Attenuation_2]
  properties {soundClassOverride=&Fort_WeaponRanged_Fire_3P_PSM}
 }
 {
  id=SoundClass_2
  children [Attenuation_0]
  properties {soundClassOverride=&Fort_WeaponRanged_Fire_3P_PSM}
 }
 {
  id=SoundClass_3
  children [Attenuation_3]
  properties {soundClassOverride=&Fort_WeaponRanged_Fire_3P_PSM}
 }
 {id=WavePlayer_0, properties {soundWaveAssetPtr=&/Carmine_Punch_Swing_01}}
 {id=WavePlayer_1, properties {soundWaveAssetPtr=&/Carmine_Punch_Swing_02}}
 {id=WavePlayer_10, properties {soundWaveAssetPtr=&/Carmine_Punch_Swing_02_opt}}
 {id=WavePlayer_11, properties {soundWaveAssetPtr=&/Carmine_Punch_Swing_01_opt}}
 {id=WavePlayer_12, properties {soundWaveAssetPtr=&/Carmine_Punch_Swing_01_distant_opt}}
 {id=WavePlayer_13, properties {soundWaveAssetPtr=&/Carmine_Punch_Swing_02_distant_opt}}
 {id=WavePlayer_14, properties {soundWaveAssetPtr=&/Carmine_Punch_Swing_03_distant_opt}}
 {id=WavePlayer_15, properties {soundWaveAssetPtr=&/Carmine_Punch_Swing_04_distant_opt}}
 {id=WavePlayer_2, properties {soundWaveAssetPtr=&/Carmine_Punch_Swing_03}}
 {id=WavePlayer_3, properties {soundWaveAssetPtr=&/Carmine_Punch_Swing_04}}
 {id=WavePlayer_4, properties {soundWaveAssetPtr=&/Carmine_Punch_Swing_01_distant}}
 {id=WavePlayer_5, properties {soundWaveAssetPtr=&/Carmine_Punch_Swing_02_distant}}
 {id=WavePlayer_6, properties {soundWaveAssetPtr=&/Carmine_Punch_Swing_03_distant}}
 {id=WavePlayer_7, properties {soundWaveAssetPtr=&/Carmine_Punch_Swing_04_distant}}
 {id=WavePlayer_8, properties {soundWaveAssetPtr=&/Carmine_Punch_Swing_04_opt}}
 {id=WavePlayer_9, properties {soundWaveAssetPtr=&/Carmine_Punch_Swing_03_opt}}
]
properties {bOverrideConcurrency=true, maxDistance=25000, volumeMultiplier=0.7, concurrencyOverrides {maxCount=2, resolutionRule=StopOldest}}
