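# Sound cue for the Carmine "GettingRock" sound, part of the Fortnite gameplay sound set.
# It plays a single wave directly from one wave player, with no randomization or mixing nodes.
# It uses the close-range Carmine weapon attenuation and routes through a third-person ranged weapon fire sound class.
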SoundCue /Game/Sounds/Fort_GamePlay_Sounds/Carmine/SC_Carmine_GettingRock_Cue
root=WavePlayer_0
nodes [{id=WavePlayer_0, properties {soundWaveAssetPtr=&/Carmine_GettingRock}}]
properties {
 attenuationSettings=/Game/Sounds/Attenuation/Fort_Weapon_Carmine_Close_Athena_Att
 maxDistance=4500
 soundClassObject=&Fort_WeaponRanged_Fire_3P_PSM
}
